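# BotSnack_Pickup_Cue is a sound cue for picking up a BotSnack item in BattleLab.
# A Random node picks one of three equally weighted pickup waves and passes it through a Modulator at the root.
# It uses weapon pickup attenuation and the loot-collect sound class, at a reduced volume.
# A concurrency override uses the PreventNew rule, so new instances are blocked once the limit is reached.
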SoundCue /Game/Sounds/BattleLab/BotSnack/BotSnack_Pickup_Cue
root=Modulator_1
nodes [
 {id=Modulator_1, children [Random_2]}
 {id=Random_2, children [WavePlayer_6,WavePlayer_7,WavePlayer_8], properties {weights [1,1,1]}}
 {id=WavePlayer_6, properties {soundWaveAssetPtr=&/BotSnack_Pickup_01}}
 {id=WavePlayer_7, properties {soundWaveAssetPtr=&/BotSnack_Pickup_02}}
 {id=WavePlayer_8, properties {soundWaveAssetPtr=&/BotSnack_Pickup_03}}
]
properties {
 attenuationSettings=&Fort_Weapon_Pickup_Att
 bOverrideConcurrency=true
 maxDistance=3500
 soundClassObject=&Fort_Loot_Collect
 volumeMultiplier=0.259
 attenuationOverrides {falloffDistance=900, stereoSpread=1000, attenuationShapeExtents {x=100}}
 concurrencyOverrides {resolutionRule=PreventNew}
}
